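# Additive, unlit, two-sided Niagara particle material, named as a hand-wind effect in the GoldEnergy set.
# A flow map distorts UVs by a Swirl dynamic parameter, then a time-driven panner scrolls a wave texture.
# View-facing falloff and vertex color mask the wave, driving particle-tinted emissive and depth-faded opacity.
# A refraction value scales with that falloff and weakens with camera distance, with shading-path switches simplifying forward and mobile opacity.
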Material /Game/Effects/Fort_Effects/GameplayPulginsAssets/GoldEnergy/M_EmeraldGlass_D_HandWind
graph {
 properties {
  bUsedWithNiagaraMeshParticles=true
  bUsedWithNiagaraSprites=true
  blendMode=BLEND_Additive
  pixelDepthOffsetMode=PDOM_Legacy
  refractionMethod=RM_IndexOfRefraction
  shadingModel=MSM_Unlit
  translucencyPass=MTP_BeforeDOF
  twoSided=true
  usedShadingModels=Unlit
 }
 nodes [
  n1 = ParticleColor()
  n2 = Abs(None=DotProduct(A=VertexNormalWS(), B=CameraVectorWS()))
  n3 = Multiply(A=Power(Base=Multiply(A=@n2, B=@n2), Exp=ScalarParameter() {defaultValue=1, parameterName=FresnelPow}), B=VertexColor())
  n4 = ComponentMask(None=TextureSample() {samplerType=SAMPLERTYPE_LinearColor, texture=/Game/Packages/Fortnite_UniqueMaterials_1/Textures/FlowMap}.RGB) {g=true, r=true}
  n5 = Add(
   A=TextureCoordinate()
   B=Multiply(A=Subtract(A=@n4) {constB=0.5}, B=DynamicParameter() {parameterIndex=1, defaultValue=<rgba 0 1 1 1>, paramNames [Swirl,1,2,3]}.Swirl)
  )
  n6 = AppendVector(A=ScalarParameter() {defaultValue=1, parameterName=Tex U mul}, B=ScalarParameter() {defaultValue=1, parameterName=Tex V mul})
  n7 = DynamicParameter() {paramNames [Time,Random,DepthFade,Ref]}
  n8 = TextureSampleParameter2D(UVs=Panner(Coordinate=Multiply(A=@n5, B=@n6), Time=Add(A=@n7.Time, B=@n7.Time)) {speedY=0.5}) {
   parameterName=Wave Texture
   samplerType=SAMPLERTYPE_LinearGrayscale
   texture=/Game/Weapons/FORT_Melee/Pickaxe_Cube_Ninja_Male/FX/T_FluidFlameTile
  }
  n9 = Multiply(A=@n3, B=@n8.R)
  n10 = Multiply(A=LinearInterpolate(A=VectorParameter() {defaultValue=<rgba 0.4948 0.4948 0.4948 1>}.RGB, B=@n1.RGB, Alpha=@n9), B=@n1.A)
  MaterialFunctionCall(Emissive=LinearInterpolate(A=@n10, B=Multiply(A=@n1.RGB, B=@n10), Alpha=ScalarParameter() {parameterName=Use Full Particle Color})) {materialFunction=&MF_FX_EmissiveAdjustments}
  Multiply()
  n11 = Multiply(A=@n1.A, B=@n9)
  n12 = ScalarParameter() {parameterName=Use Full Particle Color}
  ShadingPathSwitch(Default=DepthFade(Opacity=LinearInterpolate(A=@n9, B=@n11, Alpha=@n12), FadeDistance=@n7.DepthFade), Forward=@n9, Mobile=@n9)
  n13 = Multiply(A=@n7.Ref, B=@n3)
  n14 = ObjectPositionWS()
  n15 = CameraPositionWS()
  n16 = Clamp(None=Divide(B=Distance(A=@n14, B=@n15)) {constA=350})
  ShadingPathSwitch(Default=LinearInterpolate(Alpha=Multiply(A=@n13, B=@n16)) {constA=1, constB=1.05}) {desc=Migrated from FeatureLevelSwitch}
 ]
}
